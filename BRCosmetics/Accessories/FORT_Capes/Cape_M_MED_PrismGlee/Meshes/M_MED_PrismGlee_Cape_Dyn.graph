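# Physics asset for the PrismGlee cape accessory (M_MED), defining the bodies and joints for the cape's dynamic bones.
# Three chains (center, left, right) of seven capsules each hang from MSTR root bodies that are angularly locked to spine_05.
# Chain joints keep twist and swing1 tight, while swing2 allows more bend and loosens over the first few links.
# Pelvis, spine, thigh and calf shapes supply body volumes alongside the cape.
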PhysicsAsset /BRCosmetics/Accessories/FORT_Capes/Cape_M_MED_PrismGlee/Meshes/M_MED_PrismGlee_Cape_Dyn
bodies {
 calf_l [{shapeType=Capsule, length=33, radius=12, shapeName=calf_l_capsule, center=<-22.1721 0 0>, rotation=<rot 90 180 0>}]
 calf_r [{shapeType=Capsule, length=33, radius=12, shapeName=calf_r_capsule, center=<22.1721 0 0>, rotation=<rot 90 0 0>}]
 dyn_cape_01 [{shapeType=Capsule, length=10.3138, radius=8.3947, shapeName=dyn_cape_01_capsule, center=<8 0 8>, rotation=<rot 90 0 0>}]
 dyn_cape_02 [{shapeType=Capsule, length=10.3138, radius=8.3947, shapeName=dyn_cape_01_capsule, center=<8 0 5>, rotation=<rot 90 0 0>}]
 dyn_cape_03 [{shapeType=Capsule, length=10.3138, radius=8.3947, shapeName=dyn_cape_01_capsule, center=<8 0 5>, rotation=<rot 90 0 0>}]
 dyn_cape_04 [{shapeType=Capsule, length=10.3138, radius=8.3947, shapeName=dyn_cape_01_capsule, center=<8 0 5>, rotation=<rot 90 0 0>}]
 dyn_cape_05 [{shapeType=Capsule, length=10.3138, radius=8.3947, shapeName=dyn_cape_01_capsule, center=<8 0 5>, rotation=<rot 90 0 0>}]
 dyn_cape_06 [{shapeType=Capsule, length=10.3138, radius=8.3947, shapeName=dyn_cape_01_capsule, center=<8 0 5>, rotation=<rot 90 0 0>}]
 dyn_cape_07 [{shapeType=Capsule, length=10.3138, radius=8.3947, shapeName=dyn_cape_01_capsule, center=<8 0 5>, rotation=<rot 90 0 0>}]
 dyn_cape_MSTR [{shapeType=Capsule, length=0.505, radius=6, shapeName=dyn_cape_r_MSTR_capsule, rotation=<rot 0 -90 0>}]
 dyn_cape_l_01 [{shapeType=Capsule, length=10.3138, radius=8.3947, shapeName=dyn_cape_01_capsule, center=<8 0 8>, rotation=<rot 90 0 0>}]
 dyn_cape_l_02 [{shapeType=Capsule, length=10.3138, radius=8.3947, shapeName=dyn_cape_01_capsule, center=<8 0 5>, rotation=<rot 90 0 0>}]
 dyn_cape_l_03 [{shapeType=Capsule, length=10.3138, radius=8.3947, shapeName=dyn_cape_01_capsule, center=<8 0 5>, rotation=<rot 90 0 0>}]
 dyn_cape_l_04 [{shapeType=Capsule, length=10.3138, radius=8.3947, shapeName=dyn_cape_01_capsule, center=<8 0 5>, rotation=<rot 90 0 0>}]
 dyn_cape_l_05 [{shapeType=Capsule, length=10.3138, radius=8.3947, shapeName=dyn_cape_01_capsule, center=<8 0 5>, rotation=<rot 90 0 0>}]
 dyn_cape_l_06 [{shapeType=Capsule, length=10.3138, radius=8.3947, shapeName=dyn_cape_01_capsule, center=<8 0 5>, rotation=<rot 90 0 0>}]
 dyn_cape_l_07 [{shapeType=Capsule, length=10.3138, radius=8.3947, shapeName=dyn_cape_01_capsule, center=<8 0 5>, rotation=<rot 90 0 0>}]
 dyn_cape_l_MSTR [{shapeType=Capsule, length=0.505, radius=0.505, shapeName=dyn_cape_r_MSTR_capsule, rotation=<rot 0 -90 0>}]
 dyn_cape_r_01 [{shapeType=Capsule, length=10.3138, radius=8.3947, shapeName=dyn_cape_01_capsule, center=<8 0 8>, rotation=<rot 90 0 0>}]
 dyn_cape_r_02 [{shapeType=Capsule, length=10.3138, radius=8.3947, shapeName=dyn_cape_01_capsule, center=<8 0 5>, rotation=<rot 90 0 0>}]
 dyn_cape_r_03 [{shapeType=Capsule, length=10.3138, radius=8.3947, shapeName=dyn_cape_01_capsule, center=<8 0 5>, rotation=<rot 90 0 0>}]
 dyn_cape_r_04 [{shapeType=Capsule, length=10.3138, radius=8.3947, shapeName=dyn_cape_01_capsule, center=<8 0 5>, rotation=<rot 90 0 0>}]
 dyn_cape_r_05 [{shapeType=Capsule, length=10.3138, radius=8.3947, shapeName=dyn_cape_01_capsule, center=<8 0 5>, rotation=<rot 90 0 0>}]
 dyn_cape_r_06 [{shapeType=Capsule, length=10.3138, radius=8.3947, shapeName=dyn_cape_01_capsule, center=<8 0 5>, rotation=<rot 90 0 0>}]
 dyn_cape_r_07 [{shapeType=Capsule, length=10.3138, radius=8.3947, shapeName=dyn_cape_01_capsule, center=<8 0 5>, rotation=<rot 90 0 0>}]
 dyn_cape_r_MSTR [{shapeType=Capsule, length=0.505, radius=0.505, shapeName=dyn_cape_r_MSTR_capsule, rotation=<rot 0 -90 0>}]
 pelvis [{shapeType=Capsule, length=10.4146, radius=12.5983, center=<-2.4887 0.035 0>}]
 spine_01 [{shapeType=Capsule, length=7.9438, radius=13.5586, center=<-1.7965 1.6318 0>}]
 spine_02 [{shapeType=Capsule, length=7.3264, radius=12.6474, center=<-1.711 0.2698 0.0001>}]
 spine_03 [{shapeType=Capsule, length=8.2107, radius=13.7031, center=<0.1445 -0.2574 0.0001>}]
 spine_04 [{shapeType=Capsule, length=9.4404, radius=14.1325, center=<4.2405 -1.3646 0.0002>}]
 spine_05 [{shapeType=Sphere, radius=16.1415, center=<-0.4501 2.6408 0>}]
 thigh_l [{shapeType=Capsule, length=33, radius=12, shapeName=thigh_l_capsule, center=<-22.1721 0 0>, rotation=<rot 90 180 0>}]
 thigh_r [{shapeType=Capsule, length=33, radius=12, shapeName=thigh_r_capsule, center=<22.1721 0 0>, rotation=<rot 90 0 0>}]
}
constraints [
 {bone1Name=dyn_cape_r_01, bone2Name=dyn_cape_r_MSTR, swing1LimitDegrees=0, swing2LimitDegrees=10, twistLimitDegrees=0}
 {bone1Name=dyn_cape_r_02, bone2Name=dyn_cape_r_01, swing1LimitDegrees=3, swing2LimitDegrees=15, twistLimitDegrees=2}
 {bone1Name=dyn_cape_r_03, bone2Name=dyn_cape_r_02, swing1LimitDegrees=3, swing2LimitDegrees=20, twistLimitDegrees=2}
 {bone1Name=dyn_cape_r_04, bone2Name=dyn_cape_r_03, swing1LimitDegrees=3, swing2LimitDegrees=20, twistLimitDegrees=2}
 {bone1Name=dyn_cape_r_05, bone2Name=dyn_cape_r_04, swing1LimitDegrees=3, swing2LimitDegrees=20, twistLimitDegrees=2}
 {bone1Name=dyn_cape_r_06, bone2Name=dyn_cape_r_05, swing1LimitDegrees=3, swing2LimitDegrees=20, twistLimitDegrees=2}
 {bone1Name=dyn_cape_r_07, bone2Name=dyn_cape_r_06, swing1LimitDegrees=3, swing2LimitDegrees=20, twistLimitDegrees=2}
 {bone1Name=dyn_cape_l_01, bone2Name=dyn_cape_l_MSTR, swing1LimitDegrees=0, swing2LimitDegrees=10, twistLimitDegrees=0}
 {bone1Name=dyn_cape_l_02, bone2Name=dyn_cape_l_01, swing1LimitDegrees=3, swing2LimitDegrees=15, twistLimitDegrees=2}
 {bone1Name=dyn_cape_l_03, bone2Name=dyn_cape_l_02, swing1LimitDegrees=3, swing2LimitDegrees=20, twistLimitDegrees=2}
 {bone1Name=dyn_cape_l_04, bone2Name=dyn_cape_l_03, swing1LimitDegrees=3, swing2LimitDegrees=20, twistLimitDegrees=2}
 {bone1Name=dyn_cape_l_05, bone2Name=dyn_cape_l_04, swing1LimitDegrees=3, swing2LimitDegrees=20, twistLimitDegrees=2}
 {bone1Name=dyn_cape_l_06, bone2Name=dyn_cape_l_05, swing1LimitDegrees=3, swing2LimitDegrees=20, twistLimitDegrees=2}
 {bone1Name=dyn_cape_l_07, bone2Name=dyn_cape_l_06, swing1LimitDegrees=3, swing2LimitDegrees=20, twistLimitDegrees=2}
 {bone1Name=dyn_cape_01, bone2Name=dyn_cape_MSTR, swing1LimitDegrees=0, swing2LimitDegrees=10, twistLimitDegrees=0}
 {bone1Name=dyn_cape_02, bone2Name=dyn_cape_01, swing1LimitDegrees=3, swing2LimitDegrees=15, twistLimitDegrees=2}
 {bone1Name=dyn_cape_03, bone2Name=dyn_cape_02, swing1LimitDegrees=3, swing2LimitDegrees=20, twistLimitDegrees=2}
 {bone1Name=dyn_cape_04, bone2Name=dyn_cape_03, swing1LimitDegrees=3, swing2LimitDegrees=20, twistLimitDegrees=2}
 {bone1Name=dyn_cape_05, bone2Name=dyn_cape_04, swing1LimitDegrees=3, swing2LimitDegrees=20, twistLimitDegrees=2}
 {bone1Name=dyn_cape_06, bone2Name=dyn_cape_05, swing1LimitDegrees=3, swing2LimitDegrees=20, twistLimitDegrees=2}
 {bone1Name=dyn_cape_07, bone2Name=dyn_cape_06, swing1LimitDegrees=3, swing2LimitDegrees=20, twistLimitDegrees=2}
 {
  bone1Name=dyn_cape_r_MSTR
  bone2Name=spine_05
  swing1LimitDegrees=45
  swing1Motion=Locked
  swing2LimitDegrees=45
  swing2Motion=Locked
  twistLimitDegrees=45
  twistMotion=Locked
 }
 {
  bone1Name=dyn_cape_l_MSTR
  bone2Name=spine_05
  swing1LimitDegrees=45
  swing1Motion=Locked
  swing2LimitDegrees=45
  swing2Motion=Locked
  twistLimitDegrees=45
  twistMotion=Locked
 }
 {
  bone1Name=dyn_cape_MSTR
  bone2Name=spine_05
  swing1LimitDegrees=45
  swing1Motion=Locked
  swing2LimitDegrees=45
  swing2Motion=Locked
  twistLimitDegrees=45
  twistMotion=Locked
 }
]
